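# Editor utility widget for level designers that finds placed actors with scale outside a Min/Max range.
# findAssetsWithClasses filters level actors by assetClasses (defaulting to tree and rock props) and lists those with any axis out of range.
# List interactions and a button select those actors in the level.
# Spin boxes keep Min below Max and persist through SaveGame; the button caching classes from Content Browser selection is collapsed.
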EditorUtilityWidgetBlueprint /Game/Blueprints/Blutilities/ForLDs/ActorScaleValidator/ActorScaleValidator
graphs [
 EventGraph {nodes [
   "On Clicked (btnCacheClassesInCB)"!() -> (OutputDelegate, then, widget:User Widget Object Reference, event:Pointer Event Structure) {is=title K2Node_ComponentBoundEvent}
   > cacheSelectedAssets()
   "On Item Left Clicked (TA_Generic01_List)"!() -> (OutputDelegate, then, isShiftDown:bool, isCtrlDown:bool, data:Object Reference, item:User Widget Object Reference) {is=title K2Node_ComponentBoundEvent}
   > n1 = getSelectedData!(execute, self:TA Generic 01 List Object Reference=Get TA_Generic01_List() -> (TA_Generic01_List:TA Generic 01 List Object Reference)) -> (then, selectedData:Array of Object References)
   > selectLevelActorsFromListData!(execute, self, Array:Array of Object References=@n1) -> (then)
   "On Clicked (btnFindActors)"!() -> (OutputDelegate, then, widget:User Widget Object Reference, event:Pointer Event Structure) {is=title K2Node_ComponentBoundEvent}
   > findAssetsWithClasses()
   n2 = "On Value Committed (SpinBoxMax)"!() -> (OutputDelegate, then, InValue:float, CommitMethod:ETextCommit Enum=Default) {is=title K2Node_ComponentBoundEvent}
   > n3 = Switch on ETextCommit!(execute, Selection:ETextCommit Enum=@n2.CommitMethod) -> (Default:Exec, OnEnter:Exec, OnUserMovedFocus:Exec, OnCleared:Exec) {is=title K2Node_SwitchEnum}
   n4 = GetValue!(self:Spin Box Object Reference=Get SpinBoxMin() -> (SpinBoxMin:Spin Box Object Reference)) -> (ReturnValue:float)
   n5 = Branch(
    execute=[@n3.OnEnter, @n3.OnUserMovedFocus]
    Condition="float <= float"!(A:double=@n2.InValue, B:double=@n4) -> (ReturnValue:bool) {is=title K2Node_CallFunction}
   )
   > n6 = SetValue!(
    execute
    self:Spin Box Object Reference=Get SpinBoxMax() -> (SpinBoxMax:Spin Box Object Reference)
    NewValue:float=float + float(A=@n4, B=0.1) {is=title K2Node_CommutativeAssociativeBinaryOperator}
   ) -> (then)
   SaveGame_Save(execute=[@n6.then, @n5.else], self:TA Base Editor Widget Object Reference)
   n7 = "On Item Selection Changed (TA_Generic01_List)"!() -> (
    OutputDelegate
    then
    selectedData:Array of Object References
    selectedWidgets:Array of User Widget Object References
    selectedIndices:Array of Integers
   ) {is=title K2Node_ComponentBoundEvent}
   > selectLevelActorsFromListData(Array=@n7.selectedData)
   "On Clicked (btnSelectAllFromList)"!() -> (OutputDelegate, then, widget:User Widget Object Reference, event:Pointer Event Structure) {is=title K2Node_ComponentBoundEvent}
   > n8 = getListData!(execute, self:TA Generic 01 List Object Reference=Get TA_Generic01_List()) -> (then, data:Array of Object References)
   > selectLevelActorsFromListData(Array=@n8)
   n9 = "On Value Committed (SpinBoxMin)"!() -> (OutputDelegate, then, InValue:float, CommitMethod:ETextCommit Enum=Default) {is=title K2Node_ComponentBoundEvent}
   > n10 = Switch on ETextCommit(Selection=@n9.CommitMethod)
   n11 = GetValue(self=Get SpinBoxMax())
   n12 = Branch(execute=[@n10.OnEnter, @n10.OnUserMovedFocus], Condition="float >= float"(A=@n9.InValue, B=@n11) {is=title K2Node_CallFunction})
   > n13 = SetValue(self=Get SpinBoxMin(), NewValue=float - float(A=@n11, B=0.1) {is=title K2Node_CallFunction})
   SaveGame_Save(execute=[@n13.then, @n12.else])
  ]}
 SaveGame_PostLoad {nodes [
   n1 = SaveGame_PostLoad!() -> (then, saveGameExisted:bool, saveGame_Object:Save Game Object Reference) {is=title K2Node_FunctionEntry}
   > PrintString(InString=load)
   > Branch(Condition=@n1)
   > n2 = Cast To ActorScaleValidator_SaveGame!(execute, Object=Get saveGame_Object() -> (saveGame_Object:Save Game Object Reference)) -> (then, CastFailed, AsActor Scale Validator Save Game:Actor Scale Validator Save Game Object Reference) {is=title K2Node_DynamicCast}
   > SetValue(self=Get SpinBoxMin(), NewValue=Get min!(self:Actor Scale Validator Save Game Object Reference=@n2) -> (min:double))
   > SetValue(self=Get SpinBoxMax(), NewValue=Get max!(self:Actor Scale Validator Save Game Object Reference=@n2) -> (max:double))
   Return Node!(execute, " ":bool=false)
  ]}
 SaveGame_PrepareToSave {nodes [
   SaveGame_PrepareToSave!() -> (then, SaveGameObject:Save Game Object Reference) {is=title K2Node_FunctionEntry}
   > PrintString(InString=save)
   > n1 = Cast To ActorScaleValidator_SaveGame(Object=Get saveGame_Object())
   > Set min!(execute, min:double=GetValue(self=Get SpinBoxMin()), self:Actor Scale Validator Save Game Object Reference=@n1) -> (then, Output_Get:double)
   > Set max!(execute, max:double=GetValue(self=Get SpinBoxMax()), self:Actor Scale Validator Save Game Object Reference=@n1) -> (then, Output_Get:double)
   Return Node(" "=false)
  ]}
 cacheSelectedAssets {nodes [
   n1 = GetSelectedAssets(execute=@n7.then)
   > n2 = For Each Loop(Array:Array of Object References=@n1) -> (Array Element:Object Reference)
   n3 = Get assetClasses() -> (assetClasses:Set of Object Class References)
   n4 = Add!(
    execute
    TargetSet:Set of Object Class References=@n3
    NewItem:"Object Class Reference (by ref)"=Get Class(Object=@n2) -> (ReturnValue:Object Class Reference)
   ) -> (then) {is=title K2Node_CallFunction}
   n5 = Get Object Path String(Object=@n2)
   n6 = GetAssetsByPackageName!(
    execute=@n2.LoopBody
    self:Asset Registry Interface=GetAssetRegistry()
    PackageName:Name=String To Name(InString=Split(SourceString=@n5, InStr=., SearchDir=FromStart))
    bIncludeOnlyOnDiskAssets:bool=false
    bSkipARFilteredAssets:bool=true
   ) -> (then, OutAssetData:Array of Asset Data Structures, ReturnValue:bool)
   PrintString(
    execute=[@n6.then, @n4.then]
    InString=BreakTopLevelAssetPath!(TopLevelAssetPath:"Top Level Asset Path Structure (by ref)"="Get (a copy)"!(Array:Array of Asset Data Structures=@n6, Dimension 1) -> (Output_PackageName:Name, Output_PackagePath:Name, Output_AssetName:Name, Output_AssetClassPath:Top Level Asset Path Structure).Output_AssetClassPath) -> (PathString:String)
   )
   cacheSelectedAssets() {is=title K2Node_FunctionEntry}
   > n7 = Clear!(execute, TargetSet:Set of Object Class References=Get assetClasses()) -> (then) {is=title K2Node_CallFunction}
   PrintString(
    execute=@n2.Completed
    InString="To String (Integer)"(InInt=Length!(TargetSet:Set of Object Class References=@n3) -> (ReturnValue:int=0) {is=title K2Node_CallFunction})
   )
  ]}
 findAssetsWithClasses {nodes [
   findAssetsWithClasses() {is=title K2Node_FunctionEntry}
   n1 = Get assetClasses()
   > n2 = Branch(Condition="integer > integer"(A=Length(TargetSet=@n1) -> (ReturnValue=0), B=0) {is=title K2Node_CallFunction})
   > clearItems(self:TA Generic 01 List Object Reference=Get TA_Generic01_List())
   > Set l_max(l_max:double=GetValue(self=Get SpinBoxMax())) -> (Output_Get:double)
   > Set l_min(l_min:double=GetValue(self=Get SpinBoxMin())) -> (Output_Get:double)
   > n3 = GetAllLevelActors(self=Get EditorActorSubsystem())
   > Set l_initialActors(l_initialActors:Array of Actor Object References=@n3) -> (Output_Get:Array of Actor Object References)
   > n4 = Sequence()
   > n5 = To Array!(execute, A:Set of Object Class References=Get assetClasses()) -> (then, Result:Array of Object Class References)
   > n6 = For Each Loop(Array:Array of Object Class References=@n5) -> (Array Element:Object Class Reference)
   > n7 = Filter by Class!(
    execute
    TargetArray:Array of Object References=Get l_initialActors() -> (l_initialActors:Array of Actor Object References)
    ObjectClass:Object Class Reference=@n6
    FilterType:EEditorScriptingFilterType Enum
   ) -> (then, ReturnValue:Array of Object References)
   > n8 = For Each Loop(Array:Array of Object References=@n7) -> (Array Element:Object Reference)
   > n9 = Cast To Actor!(execute, Object=@n8) -> (then, CastFailed, AsActor:Actor Object Reference) {is=title K2Node_DynamicCast}
   > Add(
    TargetArray:Array of Actor Object References=Get l_finalActors() -> (l_finalActors:Array of Actor Object References)
    NewItem:"Actor Object Reference (by ref)"=@n9
   ) {is=title K2Node_CallArrayFunction}
   n10 = For Each Loop(Exec=@n4.then_1, Array:Array of Actor Object References=Get l_finalActors()) -> (Array Element:Actor Object Reference)
   > Set l_currentActor(l_currentActor:Actor Object Reference=@n10) -> (Output_Get:Actor Object Reference)
   n11 = BreakVector(InVec=GetActorScale3D!(self:Actor Object Reference=@n10) -> (ReturnValue:Vector))
   n12 = Get l_min() -> (l_min:double)
   n13 = Get l_max() -> (l_max:double)
   n14 = "In Range (Float)"!(Value:double=@n11, Min:double=@n12, Max:double=@n13, InclusiveMin:bool=true, InclusiveMax:bool=true) -> (ReturnValue:bool)
   > PrintString(InString="To String (Boolean)"(InBool=@n14))
   > n15 = Branch(Condition=AND Boolean!(
     A=@n14
     B="In Range (Float)"(Value=@n11.Y, Min=@n12, Max=@n13, InclusiveMin=true, InclusiveMax=true)
     C:bool="In Range (Float)"(Value=@n11.Z, Min=@n12, Max=@n13, InclusiveMin=true, InclusiveMax=true)
    ) -> (ReturnValue))
   n16 = Construct Object from Class(execute=@n15.else, Class=/Game/Blueprints/Blutilities/ForTechArt/Data/TA_GenericListItemData.TA_GenericListItemData_C, self=Self-Reference()) -> (ReturnValue:TA Generic List Item Data Object Reference)
   n17 = Get l_currentActor() -> (l_currentActor:Actor Object Reference)
   n18 = GetActorScale3D(self=Get l_currentActor())
   > Set displayString!(
    execute
    displayString:String=Append!(A=GetDisplayName(Object=@n17), B=" :: ", C:String="To String (Vector)"!(InVec:Vector=@n18) -> (ReturnValue:String)) -> (ReturnValue)
    self:TA Generic List Item Data Object Reference=@n16
   ) -> (then, Output_Get:String)
   > Set actors!(
    execute
    actors:Array of Actor Object References=Make Array("[0]":Actor Object Reference=Get l_currentActor()) -> (Array:Array of Actor Object References)
    self:TA Generic List Item Data Object Reference=@n16
   ) -> (then, Output_Get:Array of Actor Object References)
   > addItem!(execute, self:TA Generic 01 List Object Reference=Get TA_Generic01_List(), dataObject:Object Reference=@n16) -> (then, newItem:User Widget Object Reference, newData:Object Reference)
   PrintString(
    execute=@n2.else
    InString="-----------------------------------------------------------------\r\n-----------------------------------------------------------------\r\nACTOR SCALE VALIDATOR: NO CLASSES CACHED WHEN TRYING TO FIND ACTORS.  PLEASE CACHE SOME ASSETS!\r\n-----------------------------------------------------------------\r\n-----------------------------------------------------------------"
   )
  ]}
 selectLevelActorsFromListData {nodes [
   n1 = selectLevelActorsFromListData!() -> (then, Array:Array of Object References) {is=title K2Node_FunctionEntry}
   > n2 = For Each Loop(Array:Array of Object References=@n1) -> (Array Element:Object Reference)
   SetSelectedLevelActors(execute=@n2.Completed, self=Get EditorActorSubsystem(), ActorsToSelect=Get l_actors() -> (l_actors:Array of Actor Object References))
   n3 = Cast To TA_GenericListItemData!(execute=@n2.LoopBody, Object=@n2) -> (then, CastFailed, AsTA Generic List Item Data:TA Generic List Item Data Object Reference) {is=title K2Node_DynamicCast}
   > Append Array!(
    execute
    TargetArray:Array of Actor Object References=Get l_actors()
    SourceArray:Array of Actor Object References=Get actors!(self:TA Generic List Item Data Object Reference=@n3) -> (actors:Array of Actor Object References)
   ) -> (then)
  ]}
]
variables [assetClasses {
  type="array<Class>"
  default [
   /Game/Building/ActorBlueprints/Prop/Parent_Tree_Apollo.Parent_Tree_Apollo_C
   /Game/Athena/Apollo/Environments/BuildingActors/Rocks/Prop_Rock_01.Prop_Rock_01_C
  ]
 }]
widgets {
 info {
  parentClass=/Game/Blueprints/Blutilities/ForTechArt/Widgets/TA_Base_EditorWidget.TA_Base_EditorWidget_C
  rootWidgetClass=&VerticalBox
  widgetCount=13
 }
 widgets [
  VerticalBox_0 {widgetClassPath=&VerticalBox, widget=VerticalBox_0}
  btnCacheClassesInCB {
   widgetClassPath=&TA_Generic01_Button
   parent=VerticalBox_0
   widget=btnCacheClassesInCB
   properties {label=Get classes to find from selection in Content Browser, visibility=Collapsed}
   slot {class=VerticalBoxSlot}
  }
  HorizontalBox {
   widgetClassPath=&HorizontalBox
   parent=VerticalBox_0
   widget=HorizontalBox
   slot {class=VerticalBoxSlot, properties {padding=<ltrb 0 10 0 0>}}
  }
  CommonTextBlock_1 {
   widgetClassPath=&CommonTextBlock
   parent=HorizontalBox
   widget=CommonTextBlock_1
   properties {
    text=Scale Min:
    font {
     bForceMonospaced=false
     bMaterialIsStencil=false
     fontMaterial=null
     fontObject=&NotoSans
     letterSpacing=0
     monospacedWidth=1
     size=12
     skewAmount=0
     typefaceFontName=NotoSans-Regular
     outlineSettings {
      bApplyOutlineToDropShadows=false
      bMiteredCorners=false
      bSeparateFillAlpha=false
      outlineMaterial=null
      outlineSize=0
      outlineColor=<rgba 0 0 0 1>
     }
    }
    shadowOffset=<0 0>
   }
   slot {class=HorizontalBoxSlot, properties {horizontalAlignment=HAlign_Left, padding=<ltrb 10 0 0 0>}}
  }
  SizeBox_0 {
   widgetClassPath=&SizeBox
   parent=HorizontalBox
   widget=SizeBox_0
   properties {bOverride_WidthOverride=true, widthOverride=113.6047}
   slot {class=HorizontalBoxSlot, properties {horizontalAlignment=HAlign_Left, padding=<ltrb 10 0 0 0>}}
  }
  SpinBoxMin {
   widgetClassPath=/Script/UMG.SpinBox
   bIsVariable=true
   parent=SizeBox_0
   widget=SpinBoxMin
   properties {
    bOverride_MaxSliderValue=true
    bOverride_MaxValue=true
    bOverride_MinSliderValue=true
    bOverride_MinValue=true
    delta=0.01
    maxSliderValue=5
    maxValue=5
    value=0.75
    foregroundColor {colorUseRule=UseColor_Specified, specifiedColor=<rgba 1 1 1 1>}
    widgetStyle {
     activeBackgroundBrush {
      bIsDynamicallyLoaded=false
      drawAs=RoundedBox
      imageType=NoImage
      mirroring=NoMirror
      resourceName=None
      resourceObject=null
      tiling=NoTile
      imageSize=<0 0>
      margin=<ltrb 0 0 0 0>
      outlineSettings {
       bUseBrushTransparency=false
       roundingType=FixedRadius
       width=1
       color {colorUseRule=UseColor_ColorTable, specifiedColor=<rgba 0 0 0 0>}
       cornerRadii=<4 4 4 4>
      }
      tintColor {colorUseRule=UseColor_Specified, specifiedColor=<rgba 0.0048 0.0048 0.0048 1>}
      uVRegion {bIsValid=false, max=<0 0>, min=<0 0>}
     }
     activeFillBrush {
      bIsDynamicallyLoaded=false
      drawAs=RoundedBox
      imageType=NoImage
      mirroring=NoMirror
      resourceName=None
      resourceObject=null
      tiling=NoTile
      imageSize=<0 0>
      margin=<ltrb 0 0 0 0>
      outlineSettings {
       bUseBrushTransparency=false
       roundingType=FixedRadius
       width=1
       color {colorUseRule=UseColor_Specified, specifiedColor=<rgba 0 0 0 0>}
       cornerRadii=<4 4 4 4>
      }
      tintColor {colorUseRule=UseColor_Specified, specifiedColor=<rgba 0.0953 0.0953 0.0953 1>}
      uVRegion {bIsValid=false, max=<0 0>, min=<0 0>}
     }
     arrowsImage {
      bIsDynamicallyLoaded=false
      drawAs=NoDrawType
      imageType=NoImage
      mirroring=NoMirror
      resourceName=None
      resourceObject=null
      tiling=NoTile
      imageSize=<0 0>
      margin=<ltrb 0 0 0 0>
      outlineSettings {
       bUseBrushTransparency=false
       roundingType=HalfHeightRadius
       width=0
       color {colorUseRule=UseColor_Specified, specifiedColor=<rgba 0 0 0 0>}
       cornerRadii=<0 0 0 0>
      }
      tintColor {colorUseRule=UseColor_Specified, specifiedColor=<rgba 1 1 1 1>}
      uVRegion {bIsValid=false, max=<0 0>, min=<0 0>}
     }
     backgroundBrush {
      bIsDynamicallyLoaded=false
      drawAs=RoundedBox
      imageType=NoImage
      mirroring=NoMirror
      resourceName=None
      resourceObject=null
      tiling=NoTile
      imageSize=<0 0>
      margin=<ltrb 0 0 0 0>
      outlineSettings {
       bUseBrushTransparency=false
       roundingType=FixedRadius
       width=1
       color {colorUseRule=UseColor_ColorTable, specifiedColor=<rgba 0 0 0 0>}
       cornerRadii=<4 4 4 4>
      }
      tintColor {colorUseRule=UseColor_Specified, specifiedColor=<rgba 0.0048 0.0048 0.0048 1>}
      uVRegion {bIsValid=false, max=<0 0>, min=<0 0>}
     }
     foregroundColor {colorUseRule=UseColor_Specified, specifiedColor=<rgba 1 1 1 1>}
     hoveredBackgroundBrush {
      bIsDynamicallyLoaded=false
      drawAs=RoundedBox
      imageType=NoImage
      mirroring=NoMirror
      resourceName=None
      resourceObject=null
      tiling=NoTile
      imageSize=<0 0>
      margin=<ltrb 0 0 0 0>
      outlineSettings {
       bUseBrushTransparency=false
       roundingType=FixedRadius
       width=1
       color {colorUseRule=UseColor_ColorTable, specifiedColor=<rgba 0 0 0 0>}
       cornerRadii=<4 4 4 4>
      }
      tintColor {colorUseRule=UseColor_Specified, specifiedColor=<rgba 0.0048 0.0048 0.0048 1>}
      uVRegion {bIsValid=false, max=<0 0>, min=<0 0>}
     }
     hoveredFillBrush {
      bIsDynamicallyLoaded=false
      drawAs=RoundedBox
      imageType=NoImage
      mirroring=NoMirror
      resourceName=None
      resourceObject=null
      tiling=NoTile
      imageSize=<0 0>
      margin=<ltrb 0 0 0 0>
      outlineSettings {
       bUseBrushTransparency=false
       roundingType=FixedRadius
       width=1
       color {colorUseRule=UseColor_Specified, specifiedColor=<rgba 0 0 0 0>}
       cornerRadii=<4 4 4 4>
      }
      tintColor {colorUseRule=UseColor_Specified, specifiedColor=<rgba 0.0953 0.0953 0.0953 1>}
      uVRegion {bIsValid=false, max=<0 0>, min=<0 0>}
     }
     inactiveFillBrush {
      bIsDynamicallyLoaded=false
      drawAs=RoundedBox
      imageType=NoImage
      mirroring=NoMirror
      resourceName=None
      resourceObject=null
      tiling=NoTile
      imageSize=<0 0>
      margin=<ltrb 0 0 0 0>
      outlineSettings {
       bUseBrushTransparency=false
       roundingType=FixedRadius
       width=1
       color {colorUseRule=UseColor_Specified, specifiedColor=<rgba 0 0 0 0>}
       cornerRadii=<4 4 4 4>
      }
      tintColor {colorUseRule=UseColor_Specified, specifiedColor=<rgba 0.0395 0.0395 0.0395 1>}
      uVRegion {bIsValid=false, max=<0 0>, min=<0 0>}
     }
     insetPadding=<ltrb 1 1 1 1>
     textPadding=<ltrb 10 3.5 10 4>
    }
   }
   slot {class=SizeBoxSlot}
  }
  CommonTextBlock_2 {
   widgetClassPath=&CommonTextBlock
   parent=HorizontalBox
   widget=CommonTextBlock_2
   properties {
    text=Scale Max:
    font {
     bForceMonospaced=false
     bMaterialIsStencil=false
     fontMaterial=null
     fontObject=&NotoSans
     letterSpacing=0
     monospacedWidth=1
     size=12
     skewAmount=0
     typefaceFontName=NotoSans-Regular
     outlineSettings {
      bApplyOutlineToDropShadows=false
      bMiteredCorners=false
      bSeparateFillAlpha=false
      outlineMaterial=null
      outlineSize=0
      outlineColor=<rgba 0 0 0 1>
     }
    }
    shadowOffset=<0 0>
   }
   slot {class=HorizontalBoxSlot, properties {horizontalAlignment=HAlign_Left, padding=<ltrb 30 0 0 0>}}
  }
  SizeBox {
   widgetClassPath=&SizeBox
   parent=HorizontalBox
   widget=SizeBox
   properties {bOverride_WidthOverride=true, widthOverride=113.6047}
   slot {class=HorizontalBoxSlot, properties {horizontalAlignment=HAlign_Left, padding=<ltrb 10 0 0 0>}}
  }
  SpinBoxMax {
   widgetClassPath=&SpinBox
   bIsVariable=true
   parent=SizeBox
   widget=SpinBoxMax
   properties {
    bOverride_MaxSliderValue=true
    bOverride_MaxValue=true
    bOverride_MinSliderValue=true
    bOverride_MinValue=true
    delta=0.01
    maxSliderValue=5
    maxValue=5
    value=1.25
    foregroundColor {colorUseRule=UseColor_Specified, specifiedColor=<rgba 1 1 1 1>}
    widgetStyle {
     activeBackgroundBrush {
      bIsDynamicallyLoaded=false
      drawAs=RoundedBox
      imageType=NoImage
      mirroring=NoMirror
      resourceName=None
      resourceObject=null
      tiling=NoTile
      imageSize=<0 0>
      margin=<ltrb 0 0 0 0>
      outlineSettings {
       bUseBrushTransparency=false
       roundingType=FixedRadius
       width=1
       color {colorUseRule=UseColor_ColorTable, specifiedColor=<rgba 0 0 0 0>}
       cornerRadii=<4 4 4 4>
      }
      tintColor {colorUseRule=UseColor_Specified, specifiedColor=<rgba 0.0048 0.0048 0.0048 1>}
      uVRegion {bIsValid=false, max=<0 0>, min=<0 0>}
     }
     activeFillBrush {
      bIsDynamicallyLoaded=false
      drawAs=RoundedBox
      imageType=NoImage
      mirroring=NoMirror
      resourceName=None
      resourceObject=null
      tiling=NoTile
      imageSize=<0 0>
      margin=<ltrb 0 0 0 0>
      outlineSettings {
       bUseBrushTransparency=false
       roundingType=FixedRadius
       width=1
       color {colorUseRule=UseColor_Specified, specifiedColor=<rgba 0 0 0 0>}
       cornerRadii=<4 4 4 4>
      }
      tintColor {colorUseRule=UseColor_Specified, specifiedColor=<rgba 0.0953 0.0953 0.0953 1>}
      uVRegion {bIsValid=false, max=<0 0>, min=<0 0>}
     }
     arrowsImage {
      bIsDynamicallyLoaded=false
      drawAs=NoDrawType
      imageType=NoImage
      mirroring=NoMirror
      resourceName=None
      resourceObject=null
      tiling=NoTile
      imageSize=<0 0>
      margin=<ltrb 0 0 0 0>
      outlineSettings {
       bUseBrushTransparency=false
       roundingType=HalfHeightRadius
       width=0
       color {colorUseRule=UseColor_Specified, specifiedColor=<rgba 0 0 0 0>}
       cornerRadii=<0 0 0 0>
      }
      tintColor {colorUseRule=UseColor_Specified, specifiedColor=<rgba 1 1 1 1>}
      uVRegion {bIsValid=false, max=<0 0>, min=<0 0>}
     }
     backgroundBrush {
      bIsDynamicallyLoaded=false
      drawAs=RoundedBox
      imageType=NoImage
      mirroring=NoMirror
      resourceName=None
      resourceObject=null
      tiling=NoTile
      imageSize=<0 0>
      margin=<ltrb 0 0 0 0>
      outlineSettings {
       bUseBrushTransparency=false
       roundingType=FixedRadius
       width=1
       color {colorUseRule=UseColor_ColorTable, specifiedColor=<rgba 0 0 0 0>}
       cornerRadii=<4 4 4 4>
      }
      tintColor {colorUseRule=UseColor_Specified, specifiedColor=<rgba 0.0048 0.0048 0.0048 1>}
      uVRegion {bIsValid=false, max=<0 0>, min=<0 0>}
     }
     foregroundColor {colorUseRule=UseColor_Specified, specifiedColor=<rgba 1 1 1 1>}
     hoveredBackgroundBrush {
      bIsDynamicallyLoaded=false
      drawAs=RoundedBox
      imageType=NoImage
      mirroring=NoMirror
      resourceName=None
      resourceObject=null
      tiling=NoTile
      imageSize=<0 0>
      margin=<ltrb 0 0 0 0>
      outlineSettings {
       bUseBrushTransparency=false
       roundingType=FixedRadius
       width=1
       color {colorUseRule=UseColor_ColorTable, specifiedColor=<rgba 0 0 0 0>}
       cornerRadii=<4 4 4 4>
      }
      tintColor {colorUseRule=UseColor_Specified, specifiedColor=<rgba 0.0048 0.0048 0.0048 1>}
      uVRegion {bIsValid=false, max=<0 0>, min=<0 0>}
     }
     hoveredFillBrush {
      bIsDynamicallyLoaded=false
      drawAs=RoundedBox
      imageType=NoImage
      mirroring=NoMirror
      resourceName=None
      resourceObject=null
      tiling=NoTile
      imageSize=<0 0>
      margin=<ltrb 0 0 0 0>
      outlineSettings {
       bUseBrushTransparency=false
       roundingType=FixedRadius
       width=1
       color {colorUseRule=UseColor_Specified, specifiedColor=<rgba 0 0 0 0>}
       cornerRadii=<4 4 4 4>
      }
      tintColor {colorUseRule=UseColor_Specified, specifiedColor=<rgba 0.0953 0.0953 0.0953 1>}
      uVRegion {bIsValid=false, max=<0 0>, min=<0 0>}
     }
     inactiveFillBrush {
      bIsDynamicallyLoaded=false
      drawAs=RoundedBox
      imageType=NoImage
      mirroring=NoMirror
      resourceName=None
      resourceObject=null
      tiling=NoTile
      imageSize=<0 0>
      margin=<ltrb 0 0 0 0>
      outlineSettings {
       bUseBrushTransparency=false
       roundingType=FixedRadius
       width=1
       color {colorUseRule=UseColor_Specified, specifiedColor=<rgba 0 0 0 0>}
       cornerRadii=<4 4 4 4>
      }
      tintColor {colorUseRule=UseColor_Specified, specifiedColor=<rgba 0.0395 0.0395 0.0395 1>}
      uVRegion {bIsValid=false, max=<0 0>, min=<0 0>}
     }
     insetPadding=<ltrb 1 1 1 1>
     textPadding=<ltrb 10 3.5 10 4>
    }
   }
   slot {class=SizeBoxSlot}
  }
  btnFindActors {
   widgetClassPath=&TA_Generic01_Button
   parent=VerticalBox_0
   widget=btnFindActors
   properties {label=Find Cached Classes in Level outside Min and Max scale.}
   slot {class=VerticalBoxSlot, properties {padding=<ltrb 0 10 0 0>}}
  }
  Border_0 {
   widgetClassPath=&Border
   parent=VerticalBox_0
   widget=Border_0
   properties {brushColor=<rgba 0.0156 0.0156 0.0156 1>, padding=<ltrb 10 10 10 10>}
   slot {class=VerticalBoxSlot, properties {padding=<ltrb 10 10 10 10>, size {sizeRule=Fill, value=1}}}
  }
  TA_Generic01_List {
   widgetClassPath=/Game/Blueprints/Blutilities/ForTechArt/Widgets/List/TA_Generic01_List
   bIsVariable=true
   parent=Border_0
   widget=TA_Generic01_List
   slot {class=BorderSlot, properties {padding=<ltrb 10 10 10 10>}}
  }
  btnSelectAllFromList {
   widgetClassPath=&TA_Generic01_Button
   parent=VerticalBox_0
   widget=btnSelectAllFromList
   properties {label=Select Actors from List}
   slot {class=VerticalBoxSlot, properties {padding=<ltrb 0 10 0 0>}}
  }
 ]
}
